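; Dynamic UI director for LimeRock that adds and removes the video player scene.
; Once playlist data is ready, it listens for the stateful PlayVideo event and switches on its enum to add or remove the scene.
; After adding, it delays for widget initialization, then rebroadcasts a client video event chosen by the Night Night flag.
; Custom events toggle that flag or trigger return-to-main-menu setup; both old and new video flows call the same listener.
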(event EventBeginPlay
  ;[1 1 1] Video Director
  (Class|FortAsyncActionCurrentPlaylistReady|PlaylistDataReadyAsync
    (:OnReady
      (Default|VideoFlow
        (:old
          (CallFunction|ListenforPlayVideoEvent))
        (:new
          (CallFunction|ListenforPlayVideoEvent))))))

(event Custom|LR_scene
  ;[1 1 1] Video Director
  (Development|PrintString "LimeRock Video - Adding Scene")
  (DynamicUIDirector|AddScene (Variables|Default|GetDynamic%20UI%20Scene%20-%20Video)) ; Add Video Player Widget
)

(event Custom|Listen%20For%20Play%20Video%20Event
  (bind _dynamic_ui_scene___video (Variables|Default|GetDynamic%20UI%20Scene%20-%20Video))
  (bind _returnvalue (DynamicUIDirector|IsSceneActive _dynamic_ui_scene___video))
  (bind _dynamic_ui_scene___video_2 (Variables|Default|GetDynamic%20UI%20Scene%20-%20Video))
  ;[1 1 1] Video Director
  (bind (_asynctaskproxy _payload) (GameplayEventRouter|StartListeningToStatefulEvent (GameplayEventRouter|GetGlobalEventRouter) "/LimeRock/Video/EventRouter/LimeRock_Event_PlayVideo.LimeRock_Event_PlayVideo")
    (:OnEventReceived
      (bind _payload_bplay_2_981609cc4850b3577d8c5f9d8b7633ff (Utilities|Struct|BreakLimeRockEventPlayVideo _payload))
      (switch Utilities|FlowControl|Switch|SwitchonLimeRock_Event_Video_Enum _payload_bplay_2_981609cc4850b3577d8c5f9d8b7633ff
        (:NewEnumerator7
          (if (not _returnvalue)
            (Development|PrintString "LimeRock - Add Video Player Widget")
            (DynamicUIDirector|AddScene _dynamic_ui_scene___video_2) ; Add Video Player Widget
            (Utilities|FlowControl|Delay 2.0) ; let the widget be added and initialize
            (if (Variables|Default|GetNight%20Night%20Enabled)
              (GameplayEventRouter|BroadcastEvent (GameplayEventRouter|GetGlobalEventRouter) (Utilities|Struct|MakeLimeRockEventPlayVideo "NewEnumerator10") 0) ; Send video event on client
              (else
                (GameplayEventRouter|BroadcastEvent (GameplayEventRouter|GetGlobalEventRouter) (Utilities|Struct|MakeLimeRockEventPlayVideo "NewEnumerator9") 0) ; Send video event on client
              ))))
        (:NewEnumerator3
          (if _returnvalue
            (Development|PrintString "LimeRock - Stop Video")
            (DynamicUIDirector|RemoveScene _dynamic_ui_scene___video_2) ; Remove Video Player Widget
          ))))))

(event Custom|NightNightDisable
  ;[1 1 1] Video Director
  (Variables|Default|SetNight%20Night%20Enabled false))

(event Custom|NightNightEnable
  ;[1 1 1] Video Director
  (Variables|Default|SetNight%20Night%20Enabled true))

(event Custom|NightNight
  ;[1 1 1] Video Director
  (Class|SpecialEventGameplayLibrary|ClientSetupReturntoMainMenuforPlayers true) ; triger night night
)

(fn NewVideoFlow?
  (return (ScalableFloat|GetValueasBool (Variables|Default|GetUse%20New%20Video%20Flow))))

(fn ConstructionScript)
