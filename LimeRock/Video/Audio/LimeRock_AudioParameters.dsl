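; LimeRock_AudioParameters manages voice-over playback by priority, keeping active components, rules and retrigger times in a map keyed by GameInstance.
; RequestVOAsset filters requests by mute, intro or cinematic state, randomized per-tag retrigger time, DataTable priority and concurrency limit, then optionally spawns sound at the Head socket.
; RegisterAudioComponent records the sound and refreshes volumes, ducking components below the highest active priority.
; OnAudioFinished refreshes the tracked components and their volumes.
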(fn DebugPrintResults
  (params (Success :type Boolean) (Sound :type Sound%20Base%20Object%20Reference) (Reason :type String))
  (if (Variables|Debug|GetDebug)
    (Development|PrintString 0 (Utilities|String|Append (Utilities|String|SelectString "Success: " "Failure: " Success) (Utilities|GetDisplayName Sound) " ---   " Reason) true true "(R=0.000000,G=0.660000,B=1.000000,A=1.000000)" 12.0)))

(fn GetPriorityRules
  (params (GameplayTag :type Gameplay%20Tag%20Structure))
  (bind _outrownames (DataTable|GetDataTableRowNames "/LimeRock/Video/Audio/DT_LimeRock_Priorities.DT_LimeRock_Priorities"))
  (for _array_element _outrownames
    (bind _returnvalue (Utilities|GetDataTableRow "/LimeRock/Video/Audio/DT_LimeRock_Priorities.DT_LimeRock_Priorities" _array_element))
    (if (== (Utilities|Struct|BreakDTLimeRockPriorities _returnvalue) (Variables|GetGameplayTag))
      (return _returnvalue)))
  (return "(TagName=\"DeckBed.VO.Priorities\")" 0 1.5 0.0 2 0.5 0.5))

(fn OnAudioFinished
  (bind _keys (Utilities|Map|Keys (Variables|Active|GetGameplayToSettings)))
  (for _array_element _keys
    (Private|RefreshAudioComponentArrays _array_element)
    (Private|RefreshAudioComponentVolumes _array_element)))

(fn RefreshAudioComponentArrays
  (params (GameInstance :type Game%20Instance%20Object%20Reference))
  (bind _localparameters (LocalVariables|Get_localParameters))
  (bind (_activevopriorities_12_d1fca9a54f21a53bbb32c8a97edb32d1 _activevopriorityrules_16_96d801bb4ee08ba8ae1501a9e2cfcc48) (Utilities|Struct|BreakLimeRockAudioPriorityParamaters _localparameters))
  (bind _output_get (LocalVariables|Set_localParameters (Utilities|Map|Find (Variables|Active|GetGameplayToSettings) GameInstance)))
  (bind _keys (Utilities|Map|Keys (Utilities|Struct|BreakLimeRockAudioPriorityParamaters _output_get)))
  (bind _output_get_2 (LocalVariables|SetKeys _keys))
  (for _array_element _output_get_2
    (Utilities|IsValid _array_element
      (:Is%20Valid
        (if (Utilities|Enum|Equal(Enum) (Audio|Components|Audio|GetPlayState _array_element) "Stopped")
          (Audio|Components|Audio|Stop _array_element)))
      (:Is%20Not%20Valid
        _))
    (Utilities|Map|Remove _activevopriorities_12_d1fca9a54f21a53bbb32c8a97edb32d1 _array_element)
    (Utilities|Map|Remove _activevopriorityrules_16_96d801bb4ee08ba8ae1501a9e2cfcc48 _array_element))
  (Utilities|Map|Add (Variables|Active|GetGameplayToSettings) (Variables|GetGameInstance) (LocalVariables|Get_localParameters)))

(fn RefreshAudioComponentVolumes
  (params (GameInstance :type Game%20Instance%20Object%20Reference))
  (bind (_volumescale_13_29a2fc894b1c317d66866f9c59475286 _fadeouttime_15_8f8a55d145988f78cbc951b765d60a87) (Utilities|Struct|BreakDTLimeRockPriorities (LocalVariables|GetDeckBedRules)))
  (bind _localparameters (LocalVariables|Get_localParameters))
  (bind (_activevopriorities_12_d1fca9a54f21a53bbb32c8a97edb32d1 _activevopriorityrules_16_96d801bb4ee08ba8ae1501a9e2cfcc48) (Utilities|Struct|BreakLimeRockAudioPriorityParamaters _localparameters))
  (bind _output_get (LocalVariables|Set_localParameters (Utilities|Map|Find (Variables|Active|GetGameplayToSettings) GameInstance)))
  (bind _values (Utilities|Map|Values (Utilities|Struct|BreakLimeRockAudioPriorityParamaters _output_get)))
  (bind _output_get_2 (LocalVariables|SetInValues _values))
  (Utilities|Array|Reverse _output_get_2)
  (LocalVariables|SetMaxValue (.MaxValue (Math|Integer|MaxOfIntArray _output_get_2)))
  (bind _output_get_3 (LocalVariables|SetMaxValueInt (- (Utilities|Array|Length _values) (+ (Math|Integer|MaxOfIntArray _output_get_2) 1))))
  (bind _values_2 (Utilities|Map|Values _activevopriorityrules_16_96d801bb4ee08ba8ae1501a9e2cfcc48))
  (LocalVariables|SetDeckBedRules (Utilities|Array|Get(acopy) _values_2 _output_get_3))
  (bind _keys (Utilities|Map|Keys _activevopriorities_12_d1fca9a54f21a53bbb32c8a97edb32d1))
  (bind _output_get_4 (LocalVariables|SetAudioComponents _keys))
  (for _array_element _output_get_4
    (Utilities|IsValid _array_element
      (:Is%20Valid
        (if (> (LocalVariables|GetMaxValue) (Utilities|Map|Find _activevopriorities_12_d1fca9a54f21a53bbb32c8a97edb32d1 _array_element))
          (Audio|Components|Audio|AdjustVolume _array_element _fadeouttime_15_8f8a55d145988f78cbc951b765d60a87 _volumescale_13_29a2fc894b1c317d66866f9c59475286 "Linear" :!class /Script/Engine.AudioComponent)
          (else
            (Audio|Components|Audio|AdjustVolume _array_element 0.0 1.0 "Linear" :!class /Script/Engine.AudioComponent))))
      (:Is%20Not%20Valid
        (Utilities|Map|Remove _activevopriorities_12_d1fca9a54f21a53bbb32c8a97edb32d1 _array_element)
        (Utilities|Map|Remove (Utilities|Struct|BreakLimeRockAudioPriorityParamaters (LocalVariables|Get_localParameters)) _array_element)))))

(fn RegisterAudioComponent
  (params (AudioComponent :type Audio%20Component%20Object%20Reference) (Priority :type DT%20Lime%20Rock%20Priorities%20Structure) (GameTime :type Float%20%28double-precision%29) (GameInstance :type Game%20Instance%20Object%20Reference))
  (bind (_tag_19_11b687304ecc834e0eed22ac85567f96 _retriggertimemin_27_69153b65429a70fdd6f100af85375703 _retriggertimemax_26_b52bd1964d0b1883d2b4ed90c4e67179) (Utilities|Struct|BreakDTLimeRockPriorities (Variables|GetPriority)))
  (bind _gameplaytosettings (Variables|Active|GetGameplayToSettings))
  (bind _value (Utilities|Map|Find _gameplaytosettings GameInstance))
  (bind (_activevopriorities_12_d1fca9a54f21a53bbb32c8a97edb32d1 _activevopriorityrules_16_96d801bb4ee08ba8ae1501a9e2cfcc48 _activevoretriggertime_20_64896af44f9130343a4bb1bdfc9d0009) (Utilities|Struct|BreakLimeRockAudioPriorityParamaters _value))
  (bind _priority (Variables|GetPriority))
  (LocalVariables|SetVOPriorities _activevopriorities_12_d1fca9a54f21a53bbb32c8a97edb32d1)
  (LocalVariables|SetPriorityRules _activevopriorityrules_16_96d801bb4ee08ba8ae1501a9e2cfcc48)
  (LocalVariables|SetRetriggerTime _activevoretriggertime_20_64896af44f9130343a4bb1bdfc9d0009)
  (Utilities|Map|Add (LocalVariables|GetRetriggerTime) _tag_19_11b687304ecc834e0eed22ac85567f96 (Utilities|Struct|MakeStructLimeRockRetriggerTime (Variables|GetGameTime) (Math|Random|RandomFloatinRange _retriggertimemin_27_69153b65429a70fdd6f100af85375703 _retriggertimemax_26_b52bd1964d0b1883d2b4ed90c4e67179)))
  (bind _audiocomponent (Variables|GetAudioComponent)
    (:then
      (Utilities|Map|Add (LocalVariables|GetPriorityRules) _audiocomponent _priority)
      (Utilities|Map|Add (LocalVariables|GetVOPriorities) (Variables|GetAudioComponent) (Utilities|Struct|BreakDTLimeRockPriorities _priority)))
    (:else
      _))
  (bind _output_get (Variables|SetPriorityStruct_ActiveVOPriorities_12_D1FCA9A54F21A53BBB32C8A97EDB32D1 (Utilities|Struct|MakeLimeRockAudioPriorityParamaters (LocalVariables|GetVOPriorities) (LocalVariables|GetPriorityRules) (LocalVariables|GetRetriggerTime))))
  (Utilities|Map|Add (Variables|Active|GetGameplayToSettings) (Variables|GetGameInstance) _output_get)
  (bind _audiocomponent_2 (Variables|GetAudioComponent)
    (:then
      (Private|RefreshAudioComponentVolumes (Variables|GetGameInstance))
      (EventDispatchers|BindEventtoOnAudioFinished _audiocomponent_2 (EventDispatchers|CreateEvent :!event OnAudioFinished)))
    (:else)))

(fn RequestVOAsset
  (params (PriorityTag :type Gameplay%20Tag%20Structure) (SoundBase :type Sound%20Base%20Object%20Reference) (SceneComponent :type Scene%20Component%20Object%20Reference) (GameTime :type Float%20%28double-precision%29) (GameInstance :type Game%20Instance%20Object%20Reference) (SpawnSound :type Boolean :default true))
  (bind _requestedrules (LocalVariables|GetRequestedRules))
  (bind _priority_22_69300b3f42bed0176e1a378975079206 (Utilities|Struct|BreakDTLimeRockPriorities _requestedrules))
  (bind _localaudioparams (LocalVariables|GetlocalAudioParams))
  (bind _activevoretriggertime_20_64896af44f9130343a4bb1bdfc9d0009 (Utilities|Struct|BreakLimeRockAudioPriorityParamaters _localaudioparams))
  (bind _prioritytag (Variables|GetPriorityTag))
  (bind (_value _returnvalue) (Utilities|Map|Find _activevoretriggertime_20_64896af44f9130343a4bb1bdfc9d0009 _prioritytag))
  (bind (_lasttimetriggered_6_d1fca9a54f21a53bbb32c8a97edb32d1 _retriggertime_8_96d801bb4ee08ba8ae1501a9e2cfcc48) (Utilities|Struct|BreakStructLimeRockRetriggerTime _value))
  (bind _intmax (LocalVariables|GetIntMax))
  (if (Math|Boolean|ORBoolean (Variables|Debug|GetMuteVO) (Variables|Default|GetIsIntro) (Variables|Default|GetIsCinematic))
    (CallFunction|DebugPrintResults false (Variables|GetSoundBase) "VO has been Muted")
    (return false 0)
    (else
      (LocalVariables|SetlocalAudioParams (Utilities|Map|Find (Variables|Active|GetGameplayToSettings) (Variables|GetGameInstance)))
      (if _returnvalue
        (if (not (> (- (Variables|GetGameTime) _lasttimetriggered_6_d1fca9a54f21a53bbb32c8a97edb32d1) _retriggertime_8_96d801bb4ee08ba8ae1501a9e2cfcc48))
          (CallFunction|DebugPrintResults false (Variables|GetSoundBase) "Did not pass retrigger time")
          (return false 0))
        (else
          _))
      (LocalVariables|SetRequestedRules (Private|GetPriorityRules (Variables|GetPriorityTag)))
      (bind _values (Utilities|Map|Values (Utilities|Struct|BreakLimeRockAudioPriorityParamaters (LocalVariables|GetlocalAudioParams))))
      (LocalVariables|SetIntValues _values)
      (LocalVariables|SetIntMax (.MaxValue (Math|Integer|MaxOfIntArray _values)))
      (if (> _intmax _priority_22_69300b3f42bed0176e1a378975079206)
        (CallFunction|DebugPrintResults false (Variables|GetSoundBase) "Priority was too Low")
        (return false 0)
        (elif (== _intmax _priority_22_69300b3f42bed0176e1a378975079206)
          (for (_array_element _array_index) (LocalVariables|GetIntValues)
            (if (== _array_element (LocalVariables|GetIntMax))
              (Math|Integer|IncrementInt (LocalVariables|GetCurrentCount)
                (:%20%20
                  (LocalVariables|SetLastInstanceOfInt _array_index)))))
          (bind _values_2 (Utilities|Map|Values (Utilities|Struct|BreakLimeRockAudioPriorityParamaters (LocalVariables|GetlocalAudioParams))))
          (if (> (Utilities|Struct|BreakDTLimeRockPriorities (Utilities|Array|Get(acopy) _values_2 (LocalVariables|GetLastInstanceOfInt))) (LocalVariables|GetCurrentCount))
            (if (Variables|GetSpawnSound)
              (bind _returnvalue_2 (Audio|SpawnSoundAttached (Variables|GetSoundBase) (Variables|GetSceneComponent) "Head"))
              (Utilities|IsValid _returnvalue_2
                (:Is%20Valid
                  (CallFunction|RegisterAudioComponent _returnvalue_2 (LocalVariables|GetRequestedRules) (Variables|GetGameTime) (Variables|GetGameInstance))
                  (CallFunction|DebugPrintResults true (Variables|GetSoundBase) (Utilities|String|Append "Time is " (Utilities|String|ToString(Float) (Variables|GetGameTime))))
                  (return true _returnvalue_2))
                (:Is%20Not%20Valid
                  (CallFunction|DebugPrintResults false (Variables|GetSoundBase) "Audio Component not Valid")
                  (return false 0)))
              (else
                (CallFunction|RegisterAudioComponent 0 (LocalVariables|GetRequestedRules) (Variables|GetGameTime) (Variables|GetGameInstance))
                (return true 0)))
            (else
              (CallFunction|DebugPrintResults false (Variables|GetSoundBase) "Too many concurrent elements")
              (return false 0)))
          (else
            _))))))

(fn ResetVariables
  (params (GameInstance :type Game%20Instance%20Object%20Reference) (IsEndPlay :type Boolean))
  (bind _gameinstance (Variables|GetGameInstance))
  (bind _gameplaytosettings (Variables|Active|GetGameplayToSettings))
  (Utilities|IsValid GameInstance
    (:Is%20Valid
      (if (Variables|GetIsEndPlay)
        (Utilities|Map|Remove _gameplaytosettings _gameinstance)
        (Utilities|Map|Remove (Variables|Active|GetGameplayToDentVibe) (Variables|GetGameInstance))
        (else
          (Utilities|Map|Add _gameplaytosettings _gameinstance)
          (Variables|Default|SetIsCinematic false))))))

(fn SetIsIntro
  (params (IsIntro :type Boolean))
  (if (!= IsIntro (Variables|Default|GetIsIntro))
    (bind _output_get (Variables|Default|SetIsIntro IsIntro))
    (Default|CallIsIntroChanged _output_get)))
