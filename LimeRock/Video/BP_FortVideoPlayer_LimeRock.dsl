(fn Close
  (Player|Stop (Variables|Default|GetSreamingVideo) true true))

(fn Create
  (params (Priority :type EFortVideoPlayerPriority_LimeRock%20Enum :default "NewEnumerator0") (bCreateAudioComponent :type Boolean) (bDebounce :type Boolean))
  (bind _sreamingvideo (Variables|Default|GetSreamingVideo))
  (bind _sreamingvideo_2 (Variables|Default|GetSreamingVideo))
  (Utilities|FlowControl|Sequence
    (:then_0
      (Utilities|FlowControl|Sequence
        (:then_0
          (bind _returnvalue (Game|ConstructObjectfromClass "/Script/FortniteGame.FortBaseStreamingVideo"))
          (Variables|Default|SetSreamingVideo _returnvalue))
        (:then_1
          (bind _returnvalue_2 (Game|ConstructObjectfromClass "/Script/FortniteGame.FortStreamMediaSource"))
          (Variables|Default|SetStreamingMediaSource _returnvalue_2))
        (:then_2
          (bind _returnvalue_3 (Game|ConstructObjectfromClass "/Script/MediaAssets.MediaTexture"))
          (Variables|Default|SetMediaTexture _returnvalue_3))))
    (:then_1
      (Player|SetPriorityInfo _sreamingvideo (Utilities|Struct|MakeEpicMediaPriorityInfo true true (Utilities|Select :Index Priority :NewEnumerator0 "Primary" :NewEnumerator1 "Secondary" :NewEnumerator2 "Music")))
      (Player|SetCreateAudioComponent _sreamingvideo bCreateAudioComponent)
      (Player|MethodDebounce _sreamingvideo bDebounce))
    (:then_2
      (bind _returnvalue_4 (Player|Init (Variables|Default|GetSreamingVideo) (Variables|Default|GetMediaTexture) 0 (Variables|Default|GetStreamingMediaSource) true))
      (LocalVariables|SetbResult _returnvalue_4))
    (:then_3
      (if (LocalVariables|GetbResult)
        (EventDispatchers|BindEventtoVideoonSuccess _sreamingvideo_2 (EventDispatchers|CreateEvent :!event HandleSuccess))
        (EventDispatchers|BindEventtoVideoonTerminalError _sreamingvideo_2 (EventDispatchers|CreateEvent :!event HandleTerminalError))
        (EventDispatchers|BindEventtoVideoonEndReached _sreamingvideo_2 (EventDispatchers|CreateEvent :!event HandleEndReached))
        (EventDispatchers|BindEventtoVideoonClosed _sreamingvideo_2 (EventDispatchers|CreateEvent :!event HandleClosed))))
    (:then_4
      (if (LocalVariables|GetbResult)
        (Variables|Default|SetId (Guid|NewGuid))
        (else
          (Variables|Default|SetId))))
    (:then_5
      (return (LocalVariables|GetbResult)))))

(fn GetId
  (return (Variables|Default|GetId)))

(fn HandleClosed
  (CallFunction|Close)
  (Default|CallOnClose self))

(fn HandleEndReached
  (Development|PrintText 0 (Utilities|Text|FormatText "NSLOCTEXT(\"[D902D1DE1A0C4DB5A27BC969608FD091]\", \"6A679B34429CA7179C1B0EB3A9D4B84A\", \"End reasched with video {0}\")" self) true true "(R=1.000000,G=0.100000,B=0.100000,A=1.000000)" 5.0)
  (CallFunction|Close)
  (Default|CallOnClose self))

(fn HandleSuccess
  (Private|Open))

(fn HandleTerminalError
  (params (Reason :type EBaseMediaTerminalErrorReason%20Enum :default "Unknown"))
  (Development|PrintText 0 (Utilities|Text|FormatText "NSLOCTEXT(\"[D902D1DE1A0C4DB5A27BC969608FD091]\", \"B97177E54A98C514A8443CBF9430B57E\", \"Terminal error with video {0}. Reason {1}\")" self (Utilities|Text|ToText Reason)) true true "(R=1.000000,G=0.100000,B=0.100000,A=1.000000)" 5.0)
  (Default|CallOnClose self))

(fn OnCloseSignature
  (params (Player :type BP%20Fort%20Video%20Player%20Lime%20Rock%20Object%20Reference)))

(fn Open
  (Player|Open (Variables|Default|GetSreamingVideo) (struct :Tracks "(Audio=0,Caption=-1,Metadata=-1,Script=-1,Subtitle=-1,Text=-1,Video=0)" :TracksByLanguage "(Video=\"\",Audio=\"\",Subtitle=\"\",Caption=\"\")" :TrackSelection "UseTrackOptionIndices" :SeekTime "+00000000.00:00:00.000000000" :SeekTimeType "RelativeToStartTime" :PlayOnOpen "UseMediaPlayerSetting" :Loop "UseMediaPlayerSetting")))

(fn StartById
  (params (MediaId :type String))
  (bind _sreamingvideo (Variables|Default|GetSreamingVideo))
  (Player|Start _sreamingvideo MediaId (Variables|Default|GetMediaTexture))
  (bind _returnvalue (Player|GetMediaPlayer _sreamingvideo))
  (Media|MediaPlayer|Pause _returnvalue))
