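# LimeRock event video widget that plays a pre-rendered cinematic, then returns players to the front end, optionally with "Night Night".
# It starts hidden and follows stateful event-router commands to play, stop or toggle Night Night.
# Play lazily builds and initializes media config; a curve-table value gates success broadcast and return-to-menu on timeout or error.
# Pre-end or deactivation runs once-only camera and audio cleanup; deactivation or failed init triggers the return.
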WidgetBlueprint /LimeRock/Video/B_FortVideoWidgetBase_LimeRock
graphs [
 EventGraph {nodes [
   n1 = GetEventRouter!(execute, Scope:Object Class Reference=/Script/FortniteGame.FortPlayspace, ContextActor:Actor Object Reference=Self-Reference()) -> (then, ReturnValue:Gameplay Event Router Component Object Reference)
   > Broadcast Event!(execute, self:Gameplay Event Router Component Object Reference=@n1, EventData:"Wildcard (by ref)", EventContext:Object Reference) -> (then, ReturnValue:bool)
   n2 = DeactivateWidget()
   SetMediaPreEndTimeSeconds!(execute=[@n30.then, @n2.then], self:Fort Video Widget Base Object Reference, InSeconds:float=2) -> (then)
   > StartListeningForVideoEvents()
   n3 = VideoOnOpenTimeout_Event() {is=title K2Node_CustomEvent}
   > n4 = Log It!(execute, self, InString:String=VideoOnOpenTimeout, bLogAsWarning:bool=true) -> (then)
   ClientSetupReturnToMainMenuForPlayers!(execute=[@n4.then, @n33.then], bGoToNightNight:bool=true, MaxReturnTime:float=0) -> (then)
   Play Video Immediately() {is=title K2Node_CustomEvent}
   > Log It(InString=PLAY VIDEO NOW, bLogAsWarning=false)
   > n5 = Sequence!(execute) -> (then_0, then_1, then_2:Exec)
   n6 = Branch(execute=@n5.then_2, Condition=Get Init Succeeded() -> (Init Succeeded:bool))
   > Log It(InString=PLAY VIDEO - make widget visible, bLogAsWarning=false)
   > SetVisibility(InVisibility=SelfHitTestInvisible)
   > n7 = Sequence!(execute) -> (then_0, then_1)
   > Branch(Condition=GetValueAsBool!(
     Input:"Scalable Float Structure (by ref)"=Get Pre Event Countdown Time Seconds() -> (Pre Event Countdown Time Seconds:Scalable Float Structure)
     Level:float=0
    ) -> (ReturnValue:bool))
   > n8 = GetBasePlayer!(execute, self:Epic Media Video Widget Base Object Reference) -> (then, ReturnValue:Epic Base Streaming Video Object Reference)
   > n9 = Assign Video on Open Timeout!(execute:Exec, self:Epic Base Streaming Video Object Reference=@n8, Delegate:"Delegate (by ref)"=@n3) -> (then:Exec) {is=title K2Node_AssignDelegate}
   n10 = VideoOnTerminalError_Event!() -> (OutputDelegate, then, Reason:EBaseMediaTerminalErrorReason Enum=Unknown) {is=title K2Node_CustomEvent}
   n11 = Assign Video on Terminal Error!(execute:Exec=@n9.then, self:Epic Base Streaming Video Object Reference=@n8, Delegate:"Delegate (by ref)"=@n10) -> (then:Exec) {is=title K2Node_AssignDelegate}
   n12 = VideoOnSuccess_Event() {is=title K2Node_CustomEvent}
   Assign Video on Success!(execute:Exec=@n11.then, self:Epic Base Streaming Video Object Reference=@n8, Delegate:"Delegate (by ref)"=@n12) -> (then:Exec) {is=title K2Node_AssignDelegate}
   Log It(execute=@n7.then_1, InString=B_FortVideoWidgetBase Activate Widget, bLogAsWarning=false)
   > ActivateWidget(self=Self-Reference())
   Log It(execute=@n6.else, InString="PLAY VIDEO - Init didn't succeed. FAILED TO PLAY!!!", bLogAsWarning=true)
   > Trigger Night Night()
   n13 = Branch(execute=@n5.then_0, Condition=Get Init Succeeded())
   > Log It(InString="SKIPPING \"Construct Media Config\" as Init Succeeded is already true!!", bLogAsWarning=false)
   Log It(execute=@n13.else, InString="About to \"Construct Media Config\"", bLogAsWarning=false)
   > n14 = Construct Object from Class(Class=/Script/FortniteUI.FortMediaVideoWidgetConfig) -> (ReturnValue:Fort Media Video Widget Config Object Reference)
   > Set Media Config(Media Config:Fort Media Video Widget Config Object Reference=@n14) -> (Output_Get:Fort Media Video Widget Config Object Reference)
   > n15 = GetMediaId!(execute, self) -> (then, Result:String)
   > Set MediaID!(execute, MediaID:String=@n15, self:Epic Media Video Widget Config Object Reference=@n14) -> (then, Output_Get:String)
   > Set MediaName!(execute, MediaName:String=FortWidgetBase, self:Epic Media Video Widget Config Object Reference=@n14) -> (then, Output_Get:String)
   > Set bAutoPlay!(execute, bAutoPlay:bool=true, self:Epic Media Video Widget Config Object Reference=@n14) -> (then, Output_Get:bool)
   > Set bAllowSkipping!(execute, bAllowSkipping:bool=true, self:Epic Media Video Widget Config Object Reference=@n14) -> (then, Output_Get:bool)
   > Set bHideControls!(execute, bHideControls:bool=true, self:Epic Media Video Widget Config Object Reference=@n14) -> (then, Output_Get:bool)
   > Set bLoop!(execute, bLoop:bool=false, self:Epic Media Video Widget Config Object Reference=@n14) -> (then, Output_Get:bool)
   > Set bNoAudio!(execute, bNoAudio:bool=false, self:Epic Media Video Widget Config Object Reference=@n14) -> (then, Output_Get:bool)
   > Log It(InString="Done with \"Construct Media Config\"", bLogAsWarning=false)
   n16 = Branch(execute=@n5.then_1, Condition=Get Init Succeeded())
   > Log It(InString="SKIPPING \"Init Fort Widget Base\" as Init Succeeded is already true!!", bLogAsWarning=false)
   Log It(execute=@n16.else, InString="About to \"Init Fort Video Widget Base\"", bLogAsWarning=false)
   n17 = Self-Reference()
   > n18 = SetConfig!(
    execute
    self:Epic Media Video Widget Base Object Reference=@n17
    InConfig:Epic Media Video Widget Config Object Reference=Get Media Config() -> (Media Config:Fort Media Video Widget Config Object Reference)
   ) -> (then, ReturnValue:bool)
   > n19 = Branch(Condition=@n18)
   > n20 = Init!(execute, self:Epic Media Video Widget Base Object Reference=@n17) -> (then, ReturnValue:bool)
   > n21 = Branch(Condition=@n20)
   > Set Init Succeeded(Init Succeeded:bool=true) -> (Output_Get:bool)
   > Log It(InString=Init - SUCCESS, bLogAsWarning=false)
   Set Init Succeeded(execute=@n21.else, Init Succeeded=false)
   > Log It(InString=Init - FAILURE, bLogAsWarning=true)
   Log It(execute=@n19.else, InString=SET CONFIG failed, bLogAsWarning=true)
   n22 = Force Deactivate Widget() {is=title K2Node_CustomEvent}
   Parent: On Deactivated(execute=[@n22.then, @n31.then]) {is=title K2Node_CallParentFunction}
   > Do Once()
   > Log It(InString=ON DEACTIVATED has been called, bLogAsWarning=false)
   > n23 = Sequence!(execute) -> (then_0, then_1, then_2:Exec)
   Do Once(execute=[@n32.then, @n23.then_0])
   > n24 = Cast To Athena_PlayerController!(
    execute
    Object=GetFirstLocalFortPlayerController!(bIncludeNonSpectators:bool=true, bIncludeSpectators:bool=false) -> (ReturnValue:Fort Player Controller Object Reference)
   ) -> (then, CastFailed, AsAthena Player Controller:Athena Player Controller Object Reference) {is=title K2Node_DynamicCast}
   > NotifyCinematicCameraSkipped(self:Fort Player Controller Athena Object Reference=@n24)
   > SetIsIntro!(
    execute
    self:Lime Rock Audio Parameters Object Reference=Get DeckBedAudioParams() -> (DeckBedAudioParams:Lime Rock Audio Parameters Object Reference)
    IsIntro:bool=true
   ) -> (then)
   > PopSoundMixModifier!(execute, InSoundMixModifier:Sound Mix Object Reference=/LimeRock/Video/Audio/LimeRock_Cine_PreRendered) -> (then)
   > Log It(InString=PRE ENDED blueprint logic finishing, bLogAsWarning=false)
   n25 = GetEventRouter(
    execute=@n23.then_1
    Scope=&FortPlayspace
    ContextActor=GetFirstLocalFortPlayerController(bIncludeNonSpectators=true, bIncludeSpectators=false)
   )
   > Broadcast Event(
    self=@n25
    EventData:Lime Rock Event Play Video Structure=Make Lime Rock Event Play Video!(bPlay_2_#1:LimeRock_Event_Video_Enum Enum=NewEnumerator3) -> (LimeRock_Event_PlayVideo:Lime Rock Event Play Video Structure) {is=title K2Node_MakeStruct}
   )
   > DeactivateWidget()
   > SetVisibility(InVisibility=Collapsed)
   > Log It(InString=ON DEACTIVATED blueprint logic finished, bLogAsWarning=false)
   Trigger Night Night(execute=@n23.then_2)
   StartListeningForVideoEvents() {is=title K2Node_CustomEvent}
   > n26 = StartListeningToStatefulEvent!(
    execute:Exec
    Target:Gameplay Event Router Component Object Reference=GetGlobalEventRouter()
    EventType:Script Struct Object Reference=/LimeRock/Video/EventRouter/LimeRock_Event_PlayVideo
    Context:Object Reference
    EventBubblingRule:EEventBubblingRule Enum
   ) -> (
    then:Exec
    AsyncTaskProxy:Async Action Start Listening To Stateful Event Object Reference
    OnEventReceived:Exec
    OnSavedEventStateExists:Exec
    OnEventStateCleared:Exec
    Payload:Lime Rock Event Play Video Structure
   ) {is=title K2Node_StartListeningToStatefulEventAsyncAction}
   n27 = Switch on LimeRock_Event_Video_Enum!(
    execute=[@n26.OnEventReceived, @n26.OnSavedEventStateExists]
    Selection:LimeRock_Event_Video_Enum Enum=Break Lime Rock Event Play Video!(LimeRock_Event_PlayVideo:"Lime Rock Event Play Video Structure (by ref)"=@n26.Payload) -> (bPlay_2_#1:LimeRock_Event_Video_Enum Enum) {is=title K2Node_BreakStruct}
   ) -> (
    NewEnumerator6:Exec
    NewEnumerator7:Exec
    NewEnumerator0:Exec
    NewEnumerator4:Exec
    NewEnumerator1:Exec
    NewEnumerator5:Exec
    NewEnumerator3:Exec
    NewEnumerator8:Exec
    NewEnumerator9:Exec
    NewEnumerator10:Exec
    NewEnumerator11:Exec
   ) {is=title K2Node_SwitchEnum}
   Set Night Night After Video?(execute=@n27.NewEnumerator10, Night Night After Video?:bool=true) -> (Output_Get:bool)
   Set Night Night After Video?(execute=@n27.NewEnumerator9, Night Night After Video?=false)
   Play Video Immediately(execute=@n27.NewEnumerator1) {is=title K2Node_CallFunction}
   Stop(execute=@n27.NewEnumerator3, self:Epic Media Video Widget Base Object Reference)
   > DeactivateWidget()
   Trigger Night Night() {is=title K2Node_CustomEvent}
   > Do Once()
   > n28 = Branch(Condition=Get Night Night After Video?() -> (Night Night After Video?:bool))
   > Log It(InString="Triggering Night Night, returning to front-end", bLogAsWarning=false)
   > Force Deactivate Widget() {is=title K2Node_CallFunction}
   > ClientSetupReturnToMainMenuForPlayers(bGoToNightNight=true, MaxReturnTime=3)
   Log It(execute=@n28.else, InString="Night Night disabled via cheat, still returning to front-end", bLogAsWarning=false)
   > Force Deactivate Widget()
   > ClientSetupReturnToMainMenuForPlayers(bGoToNightNight=false, MaxReturnTime=3)
   n29 = Event PreConstruct()
   > Parent: PreConstruct(IsDesignTime=@n29.IsDesignTime)
   Event Construct()
   > Parent: Construct()
   > n30 = SetVisibility(InVisibility=Hidden)
   n31 = Event On Deactivated()
   Event PreEnded() {is=title K2Node_Event}
   > Parent: PreEnded() {is=title K2Node_CallParentFunction}
   > n32 = Log It(InString=PRE ENDED has been called, bLogAsWarning=false)
   Broadcast Stateful Event!(
    execute=@n12.then
    self:Gameplay Event Router Component Object Reference=GetGlobalEventRouter()
    EventData:Lime Rock Event Play Video Structure=Make Lime Rock Event Play Video(bPlay_2_#1=NewEnumerator11)
    EventContext:Object Reference
   ) -> (then, ReturnValue:bool)
   n33 = Log It(
    execute=@n10.then
    InString=Append(A="VideoOnTerminalError ", B=Enum to String!(Enumerator:EBaseMediaTerminalErrorReason Enum=@n10.Reason) -> (ReturnValue:String))
    bLogAsWarning=true
   )
  ]}
 GetMediaId {nodes [
   GetMediaId() {is=title K2Node_FunctionEntry}
   > Return Node!(execute, Result:String=Get LimeRockMediaID() -> (LimeRockMediaID:String))
  ]}
 Log It {nodes [
   n1 = Log It!() -> (then, InString:String, bLogAsWarning:bool=~) {is=title K2Node_FunctionEntry}
   > FortShippingLog!(
    execute
    WorldContextObject:Object Reference
    InString:String=Append(A="[Limerock Video - B_FortVideoWidgetBase_LimeRock] ", B=@n1)
    bLogAsWarning:bool=@n1.bLogAsWarning
   ) -> (then)
  ]}
]
variables [
 BP_AutoPlay {type=boolean}
 DeckBedAudioParams {type=LimeRock_AudioParameters_C, default=null}
 Hide HUD Tag {type=GameplayTagContainer, default {gameplayTags [{tagName=Hud.Athena.HideOnCinematic}]}}
 Init Succeeded {type=boolean}
 LimeRockMediaID {type=string, default=wTMhoZPKcXwiGvlATF}
 Media Config {type=FortMediaVideoWidgetConfig, default=null}
 Night Night After Video? {type=boolean, default=true}
 Pre Event Countdown Time Seconds {
  type=ScalableFloat
  default {registryType=None, value=1, curve {curveTable=/LimeRock/DataTables/LimeRock_GameData, rowName=TriggerNiteNiteOnVideoFailure}}
 }
]
widgets {
 info {
  inheritedWidgetCount=0
  parentClass=/Game/UI/Foundation/Movie/B_FortVideoWidgetBase.B_FortVideoWidgetBase_C
  rootWidgetClass=null
  widgetCount=0
 }
 widgets []
}
